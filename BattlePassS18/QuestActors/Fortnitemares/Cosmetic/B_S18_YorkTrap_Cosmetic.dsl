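; Cosmetic York trap actor for the Season 18 Fortnitemares quest actors, which animates, rescales and then removes itself.
; BeginPlay starts the Hide timeline, whose scale track drives uniform actor scale on each update.
; An Animation event on the same timeline plays the non-looping York Eerie open/close animation on the skeletal mesh, and the Finished output destroys the actor.
; Overlap, Tick and the construction script are empty.
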(event EventBeginPlay
  (bind (_direction _scale) (Timeline|Hide PlayFromStart)
    (:Update
      (Transformation|SetActorScale3D self (Math|Vector|MakeVector _scale _scale _scale)))
    (:Finished
      (Actor|DestroyActor))
    (:Animation
      (Components|Animation|PlayAnimation (Variables|Default|GetSkeletalMesh) "/Game/Animation/Game/Gadgets/York/York_Eerie_OpenClose_M.York_Eerie_OpenClose_M" false))))

(event Collision|EventActorBeginOverlap
  (params (OtherActor :type Actor%20Object%20Reference)))

(event EventTick
  (params (DeltaSeconds :type Float%20%28single-precision%29)))

(fn ConstructionScript)
